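# Conversation database for the NPC Kit, defining Kit's greeting dialogue when a player talks to the character.
# The entry point leads to one speech node with a default line, a first-ever-meeting line prioritized through a curve row, and lines gated on the player's character cosmetic.
# The speech node then branches through link nodes into the shared NPC Library quests and basic services conversations.
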ConversationDatabase /NPCLibrary/NPCs/Kit/NPCLibrary_Conversation_Kit
entry_points [{tag=Conversation.Entry.Kit, destinations [#1]}]
nodes [
 ConversationEntryPointNode_7 {
  id=#1
  class=ConversationEntryPointNode
  outputs [#2]
  properties {entryTag {tagName=Conversation.Entry.Kit}}
 }
 Speech_C_4 {
  id=#2
  class=Speech_C
  outputs [#3, #4]
  properties {
   nodeName=""
   generalConfig {
    defaultMessage="I could build you a mech suit, too. You could be so TALL!"
    contextualMessages [
     {
      message=Today I did parkour in my mech suit and it. Was. AWESOME.
      contextRequirements [{requirement {class=FortControllerRequirement_IsFirstEverConversationWithNPC}}]
      priority {curve {rowName=Tandem.Speech.Priority.FirstEverInteraction}}
     }
     {
      message=Did you talk to her? Using 'I feel' statements??
      contextRequirements [{requirement {
         class=FortControllerRequirement_HasCID
         properties {allowedCIds [&CID_693_Athena_Commando_M_BuffCat]}
        }}]
      priority {value=0, curve {curveTable=null, rowName=None}}
     }
     {
      message=Next I'm gonna make an even bigger suit with four arms and six legs and eight heads...
      contextRequirements [{requirement {
         class=FortControllerRequirement_HasCID
         properties {allowedCIds [/BRCosmetics/Athena/Items/Cosmetics/Characters/CID_286_Athena_Commando_F_NeonCat]}
        }}]
      priority {value=0, curve {curveTable=null, rowName=None}}
     }
     {
      message="Wow! Did you BUILD those hands, or... No? Aw."
      contextRequirements [{requirement {class=FortControllerRequirement_HasCID, properties {allowedCIds [&Character_Bites]}}}]
      priority {value=0, curve {curveTable=null, rowName=None}}
     }
     {
      message="Oh man, that suit is SO COOL. I wonder if I could build one..."
      contextRequirements [{requirement {class=FortControllerRequirement_HasCID, properties {allowedCIds [&Character_BadBear]}}}]
      priority {value=0, curve {curveTable=null, rowName=None}}
     }
     {
      message="You're so cool! Dad won't let me set ANYTHING on fire."
      contextRequirements [{requirement {class=FortControllerRequirement_HasCID, properties {allowedCIds [&Character_ChillCat]}}}]
      priority {value=0, curve {curveTable=null, rowName=None}}
     }
     {
      message=I bet I'll be on TV someday too.
      contextRequirements [{requirement {class=FortControllerRequirement_HasCID, properties {allowedCIds [&Character_PinkSpike]}}}]
      priority {value=0, curve {curveTable=null, rowName=None}}
     }
     {
      message=(Kit hisses.)
      contextRequirements [{requirement {class=FortControllerRequirement_HasCID, properties {allowedCIds [&Character_TheHerald]}}}]
      priority {value=0, curve {curveTable=null, rowName=None}}
     }
     {
      message=I know you're not a real spider.
      contextRequirements [{requirement {class=FortControllerRequirement_HasCID, properties {allowedCIds [&Character_RoseDust]}}}]
      priority {value=0, curve {curveTable=null, rowName=None}}
     }
    ]
   }
  }
 }
 ConversationLinkNode_0 {id=#4, class=ConversationLinkNode, properties {remoteEntryTag {tagName=Conversation.Entry.BasicServices}}}
 ConversationLinkNode_1 {id=#3, class=ConversationLinkNode, properties {remoteEntryTag {tagName=Conversation.Entry.NPCLibrary.Quests}}}
]
